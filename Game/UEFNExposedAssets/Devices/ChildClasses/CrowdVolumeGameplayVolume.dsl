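; CrowdVolumeGameplayVolume is a Blueprint gameplay volume, apparently a child class used by a crowd volume device, that cleans up its playspace when removed.
; On EndPlay, if the reason is Destroyed, it gets its playspace and passes it to PlayspaceLibrary DestroyPlayspace.
; Other EndPlay reasons do nothing, and BeginPlay, actor overlap and the construction script carry no logic.
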(event EventBeginPlay)

(event Collision|EventActorBeginOverlap
  (params (OtherActor :type Actor%20Object%20Reference)))

(event EventEndPlay
  (params (EndPlayReason :type EEndPlayReason%20Enum :default "Destroyed"))
  (if (Utilities|Enum|Equal(Enum) EndPlayReason "Destroyed")
    (Class|PlayspaceLibrary|DestroyPlayspace (CallFunction|GetPlayspace))))

(fn ConstructionScript)
